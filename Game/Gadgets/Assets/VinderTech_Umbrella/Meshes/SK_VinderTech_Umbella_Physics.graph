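# Physics asset for the VinderTech Umbrella gadget's skeletal mesh, giving it simplified collision and physics bodies.
# A long capsule covers the two-handed melee handle bone, and a wide, flat box covers the blades master bone.
# A single constraint joins the blades to the handle with swing and twist all free, so its angular limit values do not restrict rotation.
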PhysicsAsset /Game/Gadgets/Assets/VinderTech_Umbrella/Meshes/SK_VinderTech_Umbella_Physics
bodies {
 Melee_TwoHanded_Handle [{shapeType=Capsule, length=117.3479, radius=7.652, center=<0.0598 0.0262 43.0827>, rotation=<rot -0.0248 -90 0.1235>}]
 blades_mstr [{shapeType=Box, extentX=107.51, extentY=107.51, extentZ=17.01, center=<0 -0.0004 -9.5324>, rotation=<rot 0 -90 0>}]
}
constraints [{
  bone1Name=blades_mstr
  bone2Name=Melee_TwoHanded_Handle
  swing1LimitDegrees=45
  swing1Motion=Free
  swing2LimitDegrees=45
  swing2Motion=Free
  twistLimitDegrees=45
  twistMotion=Free
 }]
